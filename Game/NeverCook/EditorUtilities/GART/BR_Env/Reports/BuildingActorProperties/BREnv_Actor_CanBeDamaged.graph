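# GART editor report filter that checks whether whitelisted building actors can be damaged, as standard ones should not be made indestructible.
# RunFilterLogicOnObject casts the object to BuildingActor and loops the class whitelist, breaking on the first parent-class match.
# For a match, it reads bCanbeDamaged by name through GetEditorProperty and returns that value as Valid.
# The attached report data recommends BuildingStaticMesh for indestructible actors.
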EditorUtilityBlueprint /Game/NeverCook/EditorUtilities/GART/BR_Env/Reports/BuildingActorProperties/BREnv_Actor_CanBeDamaged
graphs [
 EventGraph {links [], nodes []}
 RunFilterLogicOnObject {nodes [
   n1 = RunFilterLogicOnObject!() -> (then, object:Object Reference) {is=title K2Node_FunctionEntry}
   > n2 = Cast To BuildingActor!(execute, Object=@n1) -> (then, CastFailed, AsBuilding Actor:Building Actor Object Reference) {is=title K2Node_DynamicCast}
   > n3 = Set l_buildingActor(l_buildingActor:Building Actor Object Reference=@n2) -> (Output_Get:Building Actor Object Reference)
   > n4 = For Each Loop with Break!(
    Exec:Exec
    Array:Array of Object Soft Class References=Get ClassWhitelist() -> (ClassWhitelist:Array of Object Soft Class References)
    Break:Exec=@n9.then
   ) -> (LoopBody:Exec, Array Element:Object Soft Class Reference, Array Index:int, Completed:Exec)
   n5 = ClassIsChildOf(
    TestClass=Get Class(Object=@n3) -> (ReturnValue:Building Actor Class Reference)
    ParentClass=Resolve Soft Reference!(Input:Object Soft Class Reference=@n4) -> (Output:Object Class Reference)
   )
   n6 = Branch(execute=@n4.Completed, Condition=@n5)
   > n7 = GetEditorProperty!(
    execute:Exec
    Object:Object Reference=Get l_buildingActor() -> (l_buildingActor:Building Actor Object Reference)
    PropertyName:Name=bCanbeDamaged
   ) -> (then:Exec, PropertyValue:bool, ReturnValue:bool)
   > n8 = Branch(Condition=@n7.ReturnValue)
   > Return Node!(
    execute
    Valid:bool=@n7
    ReportData:Struct GART Report Data Structure=Make Struct GART Report Data(
     DisplayTitle_20_#1=Building Actor
     DisplayDescShort_29_#2=Can be Damaged: False
     DisplayDescLong_30_#3="No standard building actor should utilize this property setting to make the actor indestructible, the BuildingStaticMesh class should be used for indestructible actors"
    )
   )
   Return Node(execute=@n8.else)
   Return Node(execute=@n6.else)
   n9 = Branch(execute=@n4.LoopBody, Condition=@n5)
   Return Node(execute=@n2.CastFailed)
  ]}
]
variables [
 Asset Counter {type=integer, default=2474}
 ClassWhitelist {
  type="array<Class>"
  default [
   &BuildingWall
   &BuildingFloor
   &BuildingStairs
   &BuildingRoof
   /Script/FortniteGame.BuildingCorner
   /Script/FortniteGame.BuildingPillar
   &BuildingProp
   /Script/FortniteGame.BuildingPropWall
  ]
 }
]
