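# Editor report filter in the GART BR environment tools that supplies the expected NavLink classes for building actor Blueprints.
# It reads each asset's NativeParentClass tag and looks up allowed NavLink names for floors, roofs, stairs and walls, exiting early for unmapped classes.
# Walls continue only if their asset name contains "Window".
# It appends "_C" to each name, stores the list as PropertyExpectedForClass, and delegates validation to the parent filter.
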EditorUtilityBlueprint /Game/NeverCook/EditorUtilities/GART/BR_Env/Reports/BuildingActorProperties/BREnv_BuildingActor_NavLink
graphs [
 EventGraph {links [], nodes []}
 RunFilterLogicOnAssetData {nodes [
   n1 = RunFilterLogicOnAssetData()
   n2 = MakeLiteralName(Value=NativeParentClass)
   n3 = To Soft Class Reference(SoftClassPath=MakeSoftClassPath(PathString=GetTagValue(InAssetData=@n1, InTagName=@n2)))
   n4 = Find(
    TargetMap:Map of Object Soft Class References to Strings=Get NavLinksByNativeClass() -> (NavLinksByNativeClass:Map of Object Soft Class References to Strings)
    Key:"Object Soft Class Reference (by ref)"=@n3
   )
   > n5 = Branch(Condition=@n4.ReturnValue)
   > n6 = Branch(Condition="Equal (SoftClassReference)"(
     A:"Object Soft Class Reference (by ref)"=@n3
     B:"Object Soft Class Reference (by ref)"=Get l_BuildingWallClass() -> (l_BuildingWallClass:Object Soft Class Reference)
    ) -> (ReturnValue:bool=~) {is=title K2Node_PromotableOperator})
   > n7 = Sequence()
   n8 = Break Asset Data(AssetData=Get AssetData())
   > n9 = Branch(Condition=Contains(SearchIn="To String (Name)"(InName=@n8.AssetName), Substring=Window))
   Return Node!(execute=@n9.else, Valid:bool, ReportData:Struct GART Report Data Structure)
   n10 = For Each Loop(Exec=[@n7.then_1, @n6.else], Array=ParseIntoArray(SourceString=@n4, Delimiter=", "))
   Set PropertyExpectedForClass(execute=@n10.Completed, PropertyExpectedForClass:Array of Strings=Get addC() -> (addC:Array of Strings)) -> (Output_Get:Array of Strings)
   > n11 = Parent: RunFilterLogicOnAssetData!(execute, AssetData:Asset Data Structure=Get AssetData(), AllowLoading:bool=Get AllowLoading() -> (AllowLoading:bool)) -> (then, Valid:bool, ReportData:Struct GART Report Data Structure) {is=title K2Node_CallParentFunction}
   > Return Node(Valid=@n11, ReportData=@n11.ReportData)
   Add(execute=@n10.LoopBody, TargetArray=Get addC(), NewItem=Append(A=@n10, B=_C)) {is=title K2Node_CallArrayFunction}
   Return Node(execute=@n5.else)
  ]}
]
variables [NavLinksByNativeClass {default {
   &BuildingFloor="NavLink_Blockout_BrokenFloor_Small, NavLink_Blockout_Floor, NavLink_Blockout_Floor_Curved_Large, NavLink_Blockout_Floor_Curved_Small, NavLink_Blockout_Floor_Diag, NavLink_Blockout_Floor_Half, NavLink_Blockout_Balcony, NavLink_Blockout_Balcony_Corner_Large, NavLink_Blockout_Balcony_Corner_Small, NavLink_Blockout_Balcony_Diag"
   &BuildingRoof="NavLink_Blockout_LowRoof_Corner_Inner, NavLink_Blockout_LowRoof_Corner_Inner_Float, NavLink_Blockout_LowRoof_Corner_Outer, NavLink_Blockout_LowRoof_Corner_Outer_Float, NavLink_Blockout_LowRoof_Joint, NavLink_Blockout_LowRoof_Joint_Float, NavLink_Blockout_LowRoof_Peak, NavLink_Blockout_LowRoof_Slant, NavLink_Blockout_LowRoof_Slant_Float, NavLink_Blockout_Roof_Corner_Inner, NavLink_Blockout_Roof_Corner_Outer, NavLink_Blockout_Roof_Joint, NavLink_Blockout_Roof_Peak, NavLink_Blockout_Roof_Peak_Center, NavLink_Blockout_Roof_Peak_Corner, NavLink_Blockout_Roof_Peak_End, NavLink_Blockout_Roof_Peak_Joint, NavLink_Blockout_Roof_Slant"
   &BuildingStairs="NavLink_Blockout_Stair_A, NavLink_Blockout_Stair_B, NavLink_Blockout_Stair_C, NavLink_Blockout_Stair_Curved, NavLink_Blockout_Stair_D"
   &BuildingWall="NavLink_WindowBothSides112cm, NavLink_WindowBothSides128cm, NavLink_WindowCenter, NavLink_WindowLeft112cm, NavLink_WindowLeft128cm, NavLink_WindowRight112cm, NavLink_WindowRight128cm, NavLink_WindowBothSides128cm_Diag, NavLink_WindowCenter_Diag"
  }}]
