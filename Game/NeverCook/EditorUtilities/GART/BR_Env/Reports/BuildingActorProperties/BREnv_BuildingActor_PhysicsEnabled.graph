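# GART editor report filter for building actor Blueprints, named for a physics-enabled check that presumably lives in its parent class.
# When the native parent matches the BuildingProp class variable, it gathers non-editor-only static mesh components through SubobjectDataSubsystem.
# It tracks the largest square root of local bounds volume and runs parent filter logic only below a fixed threshold.
# Assets with other native parents go straight to the parent logic.
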EditorUtilityBlueprint /Game/NeverCook/EditorUtilities/GART/BR_Env/Reports/BuildingActorProperties/BREnv_BuildingActor_PhysicsEnabled
graphs [
 EventGraph {links [], nodes []}
 GetBlueprintSmComponents {nodes [
   n1 = GetBlueprintSmComponents!() -> (then, AssetData:Asset Data Structure) {is=title K2Node_FunctionEntry}
   n2 = Is Valid Asset Data!(InAssetData:"Asset Data Structure (by ref)"=@n1) -> (ReturnValue:bool)
   > n3 = Branch(Condition=@n2)
   > n4 = GetBlueprintAsset(Object=GetAsset(InAssetData=@n1))
   > n5 = Gather Subobject Data For Blueprint(self=Get SubobjectDataSubsystem(), Context=@n4)
   > n6 = For Each Loop(Array:Array of Subobject Data Handle Structures=@n5) -> (Array Element:Subobject Data Handle Structure)
   n7 = Get l_StaticMeshComponents() -> (l_StaticMeshComponents:Array of Static Mesh Component Object References)
   n8 = "integer > integer"(A=Length(TargetArray:Array of Static Mesh Component Object References=@n7))
   n9 = Branch(execute=@n6.Completed, Condition=@n8)
   > Return Node!(execute, StaticMeshComponents:Array of Static Mesh Component Object References=@n7, SMCompFound:bool=@n8)
   Return Node(execute=@n9.else, SMCompFound=@n8)
   n10 = GetAssociatedObject!(Data:"Subobject Data Structure (by ref)"=GetData(DataHandle=@n6)) -> (ReturnValue:Object Reference)
   Is Valid(exec=@n6.LoopBody, InputObject=@n10)
   > n11 = Cast To StaticMeshComponent(Object=@n10)
   > n12 = Branch(Condition=Get bIsEditorOnly!(self:Actor Component Object Reference=@n11) -> (bIsEditorOnly:bool))
   Add(
    execute=@n12.else
    TargetArray:Array of Static Mesh Component Object References=@n7
    NewItem:"Static Mesh Component Object Reference (by ref)"=@n11
   ) {is=title K2Node_CallArrayFunction}
   Return Node(execute=@n3.else, SMCompFound=@n2)
  ]}
 RunFilterLogicOnAssetData {nodes [
   n1 = RunFilterLogicOnAssetData()
   n2 = MakeSoftClassPath(PathString=GetTagValue(InAssetData=Get AssetData(), InTagName=MakeLiteralName(Value=NativeParentClass)))
   > n3 = Branch(Condition="Equal (SoftClassReference)"(
     A:"Object Soft Class Reference (by ref)"=To Soft Class Reference(SoftClassPath=@n2)
     B:"Object Soft Class Reference (by ref)"=Get l_BuildingProp() -> (l_BuildingProp:Object Soft Class Reference)
    ) -> (ReturnValue:bool=~) {is=title K2Node_PromotableOperator})
   > n4 = Sequence()
   > n5 = GetBlueprintSmComponents!(execute, self, AssetData:Asset Data Structure=Get AssetData()) -> (then, StaticMeshComponents:Array of Static Mesh Component Object References, SMCompFound:bool) {is=title K2Node_CallFunction}
   > n6 = For Each Loop(Array:Array of Static Mesh Component Object References=@n5) -> (Array Element:Static Mesh Component Object Reference)
   n7 = GetLocalBounds!(self:Static Mesh Component Object Reference=@n6) -> (Min:Vector, Max:Vector)
   n8 = BreakVector(InVec=vector - vector(A=@n7.Max, B=@n7))
   n9 = Sqrt!(A:double=float * float!(A=@n8, B=@n8.Y, C:double=@n8.Z) -> (ReturnValue) {is=title K2Node_PromotableOperator}) -> (ReturnValue:double)
   > Branch(Condition="float > float"(A=@n9, B=Get l_LargestComponent() -> (l_LargestComponent:float)))
   > Set l_LargestComponent(l_LargestComponent:float=@n9) -> (Output_Get:float)
   n10 = Branch(execute=@n4.then_1, Condition="float < float"(A=Get l_LargestComponent(), B=3171))
   > n11 = Parent: RunFilterLogicOnAssetData!(execute, AssetData:Asset Data Structure=Get AssetData(), AllowLoading:bool=Get AllowLoading() -> (AllowLoading:bool)) -> (then, Valid:bool, ReportData:Struct GART Report Data Structure) {is=title K2Node_CallParentFunction}
   > Return Node!(execute, Valid:bool=@n11, ReportData:Struct GART Report Data Structure=@n11.ReportData)
   Return Node(execute=@n10.else)
   n12 = Parent: RunFilterLogicOnAssetData(execute=@n3.else, AssetData=@n1, AllowLoading=@n1.AllowLoading)
   > Return Node(Valid=@n12, ReportData=@n12.ReportData)
  ]}
]
variables []
