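# GART editor report rule for BR environment building actors that supplies the attachment types expected for building props, going by its name and defaults.
# It reads each asset's NativeParentClass tag and looks that class up in TypeByNativeClass, which lists only BuildingProp.
# On a match, it splits the mapped string into PropertyExpectedForClass and returns the parent filter's result.
# Unmapped classes return default, invalid output without running the parent logic.
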EditorUtilityBlueprint /Game/NeverCook/EditorUtilities/GART/BR_Env/Reports/BuildingActorProperties/BREnv_BuildingActor_PropAttachmentType
graphs [
 EventGraph {links [], nodes []}
 RunFilterLogicOnAssetData {nodes [
   n1 = RunFilterLogicOnAssetData()
   n2 = GetTagValue(InAssetData=@n1, InTagName=MakeLiteralName(Value=NativeParentClass))
   n3 = Find(
    TargetMap:Map of Object Soft Class References to Strings=Get TypeByNativeClass() -> (TypeByNativeClass:Map of Object Soft Class References to Strings)
    Key:"Object Soft Class Reference (by ref)"=To Soft Class Reference(SoftClassPath=MakeSoftClassPath(PathString=@n2))
   )
   > n4 = Branch(Condition=@n3.ReturnValue)
   > Set PropertyExpectedForClass(PropertyExpectedForClass:Array of Strings=ParseIntoArray(SourceString=@n3, Delimiter=", ")) -> (Output_Get:Array of Strings)
   > n5 = Parent: RunFilterLogicOnAssetData!(execute, AssetData:Asset Data Structure=@n1, AllowLoading:bool=@n1.AllowLoading) -> (then, Valid:bool, ReportData:Struct GART Report Data Structure) {is=title K2Node_CallParentFunction}
   > Return Node!(execute, Valid:bool=@n5, ReportData:Struct GART Report Data Structure=@n5.ReportData)
   Return Node(execute=@n4.else)
  ]}
]
variables [TypeByNativeClass {default {&BuildingProp="ATTACH_Floor, ATTACH_Ceiling, ATTACH_Wall, ATTACH_Wall_Then_Floor"}}]
