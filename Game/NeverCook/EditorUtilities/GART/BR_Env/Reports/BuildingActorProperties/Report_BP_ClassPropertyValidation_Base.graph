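# Base GART editor report that validates one configurable editor property on assets whose native parent is a listed building actor class.
# RunFilterLogicOnAssetData matches the NativeParentClass tag against test and negative-test class lists, then looks up PropertyName in the asset's editor property map.
# Normal classes must hold an expected value, while negative-test classes must not and get a different severity.
# Mismatches add errors listing those values under a Building Actor report entry.
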EditorUtilityBlueprint /Game/NeverCook/EditorUtilities/GART/BR_Env/Reports/BuildingActorProperties/Report_BP_ClassPropertyValidation_Base
graphs [
 CheckIfBuildingActorClass {nodes [
   CheckIfBuildingActorClass!() -> (then, AssetData:Asset Data Structure) {is=title K2Node_FunctionEntry}
   n1 = MakeLiteralName(Value=NativeParentClass)
   n2 = Get AssetData()
   n3 = To Soft Class Reference(SoftClassPath=MakeSoftClassPath(PathString=GetTagValue(InAssetData=@n2, InTagName=@n1)))
   n4 = Contains Item(
    TargetArray:Array of Object Soft Class References=Get BuildingActorClasses_NegativeTest() -> (BuildingActorClasses_NegativeTest:Array of Object Soft Class References)
    ItemToFind:"Object Soft Class Reference (by ref)"=@n3
   )
   n5 = Get BuildingActorClasses_Test() -> (BuildingActorClasses_Test:Array of Object Soft Class References)
   > Return Node!(execute, IsBuildngActorClass:bool=OR Boolean(A=Contains Item(TargetArray=@n5, ItemToFind=@n3), B=@n4), isNegativeTest:bool=@n4)
   n6 = MakeLiteralName(Value=ParentClass)
   n7 = Get AssetData()
   To Soft Class Reference(SoftClassPath=MakeSoftClassPath(PathString=GetTagValue(InAssetData=@n7, InTagName=@n6)))
  ]}
 EventGraph {links [], nodes []}
 RunFilterLogicOnAssetData {nodes [
   n1 = PrintString(InString=Append(A=Get PropertyName() -> (PropertyName:String), B=" not found in property map"))
   Return Node!(execute=[@n1.then, @n13.else], Valid:bool, ReportData:Struct GART Report Data Structure)
   n2 = RunFilterLogicOnAssetData()
   > n3 = Set l_currentAssetData(l_currentAssetData:Asset Data Structure=@n2) -> (Output_Get:Asset Data Structure)
   > n4 = CheckIfBuildingActorClass!(execute, self, AssetData:Asset Data Structure=@n3) -> (then, IsBuildngActorClass:bool, isNegativeTest:bool) {is=title K2Node_CallFunction}
   > n5 = Branch(Condition=@n4)
   > n6 = Set l_isNegativeTest(l_isNegativeTest:bool=@n4.isNegativeTest) -> (Output_Get:bool)
   > n7 = Branch(Condition=@n6)
   > n8 = Set Severity(Severity:Enum_GART_ReportSeverity Enum=NewEnumerator1) -> (Output_Get:Enum_GART_ReportSeverity Enum)
   n9 = Sequence(execute=[@n7.else, @n8.then])
   > n10 = GetEditorPropertyMap!(
    execute
    AssetData:Asset Data Structure=Get l_currentAssetData() -> (l_currentAssetData:Asset Data Structure)
    __WorldContext:Object Reference=Self-Reference()
   ) -> (then, Out Editor Properties:Map of Strings to Strings)
   > n11 = Set l_currentPropertyMap(l_currentPropertyMap:Map of Strings to Strings=@n10) -> (Output_Get:Map of Strings to Strings)
   n12 = Find(TargetMap=@n11, Key=Get PropertyName())
   > Set l_testedPropertyValue(l_testedPropertyValue:String=@n12) -> (Output_Get:String)
   > n13 = Branch(Condition=@n12.ReturnValue)
   > n14 = Branch(Condition=Get l_isNegativeTest() -> (l_isNegativeTest:bool))
   n15 = Contains Item(
    TargetArray:Array of Strings=Get PropertyExpectedForClass() -> (PropertyExpectedForClass:Array of Strings)
    ItemToFind:"String (by ref)"=Get l_testedPropertyValue() -> (l_testedPropertyValue:String)
   )
   > n16 = Branch(Condition=@n15)
   n17 = GetTagValue(InAssetData=Get l_currentAssetData(), InTagName=MakeLiteralName(Value=NativeParentClass))
   n18 = Mid!(SourceString:String=@n17, Start:int=int + int(A=FindSubstring(SearchIn=@n17, Substring=.Building)), Count:int=Len(S=@n17)) -> (ReturnValue:String)
   > n19 = AddError(NewError=Append!(A=This asset is a ', B=@n18, C:String=" class, unexpected property settings for this class include:") -> (ReturnValue))
   n20 = Get PropertyExpectedForClass()
   n21 = For Each Loop(Exec=[@n19.then, @n25.then], Array=@n20)
   AddError(execute=@n21.Completed, NewError=Get l_expectedPropertyList() -> (l_expectedPropertyList:String), Depth=1)
   n22 = Branch(execute=@n21.LoopBody, Condition="Equal (Integer)"(B=@n21.Array Index))
   > Set l_expectedPropertyList(l_expectedPropertyList:String=@n21) -> (Output_Get:String)
   n23 = Branch(execute=@n22.else, Condition="Equal (Integer)"(A=@n21.Array Index, B=Length(TargetArray:Array of Strings=@n20)))
   > Set l_expectedPropertyList(l_expectedPropertyList=Append(A=Get l_expectedPropertyList(), B=", or ", C=@n21))
   Set l_expectedPropertyList(execute=@n23.else, l_expectedPropertyList=Append(A=Get l_expectedPropertyList(), B=", ", C=@n21))
   Return Node(execute=[@n24.then, @n16.else])
   n24 = Branch(execute=@n14.else, Condition=@n15)
   n25 = AddError(execute=@n24.else, NewError=Append(A=This asset is a ', B=@n18, C=" class, valid property settings for this class include:"))
   n26 = ToLower!(SourceString:String=Get l_testedPropertyValue()) -> (ReturnValue:String)
   n27 = Get PropertyName()
   Return Node(
    execute=@n9.then_1
    Valid=IsEmpty(InString=GetCurrentErrorString!() -> (CurrentErrorString:String))
    ReportData=Make Struct GART Report Data(
     DisplayTitle_20_#1=Building Actor
     DisplayDescShort_29_#2=Append!(A=@n27, B=" set to '", C=@n26, D:String='.) -> (ReturnValue)
     Severity_46_#3=Get Severity() -> (Severity:Enum_GART_ReportSeverity Enum)
    )
   )
   Return Node(execute=@n5.else)
  ]}
]
variables [
 BuildingActorClasses_NegativeTest {type="array<Class>", default []}
 BuildingActorClasses_Test {
  type="array<Class>"
  default [
   &BuildingWall
   &BuildingFloor
   &BuildingStairs
   &BuildingRoof
   &BuildingProp
  ]
 }
 PropertyExpectedForClass {type="array<string>", default []}
 PropertyName {type=string, default=""}
 Severity {type=Enum_GART_ReportSeverity, default=Note}
]
